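# Sound cue for small-boot scuff foley on biped cosmetic companions, played quietly in the creature pets sound class.
# A quality-level root sends its first two levels to an equal-weight random pick of eight scuff waves.
# The last two levels share a reduced random pool that reuses only the first four waves.
# Companion foley attenuation and a shared biped footstep concurrency set govern playback.
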SoundCue /CosmeticCompanions/Assets/Audio_Shared/Foley/Biped/Boots/Boots_Small/SC_MImosa_Biped_Boots_Small_Scuff
root=QualityLevel_0
nodes [
 {id=QualityLevel_0, children [Random_2,Random_2,Random_3,Random_3]}
 {
  id=Random_2
  children [WavePlayer_11, WavePlayer_12, WavePlayer_13, WavePlayer_14, WavePlayer_15, WavePlayer_16, WavePlayer_20, WavePlayer_21]
  properties {weights [1,1,1,1,1,1,1,1]}
 }
 {id=Random_3, children [WavePlayer_22, WavePlayer_23, WavePlayer_24, WavePlayer_25], properties {weights [1,1,1,1]}}
 {
  id=WavePlayer_11
  properties {soundWaveAssetPtr=/CosmeticCompanions/Assets/Audio_Shared/Foley/Biped/Boots/Boots_Small/Wavs/Mimosa_Biped_Boots_Small_Scuff_01}
 }
 {
  id=WavePlayer_12
  properties {soundWaveAssetPtr=/CosmeticCompanions/Assets/Audio_Shared/Foley/Biped/Boots/Boots_Small/Wavs/Mimosa_Biped_Boots_Small_Scuff_02}
 }
 {
  id=WavePlayer_13
  properties {soundWaveAssetPtr=/CosmeticCompanions/Assets/Audio_Shared/Foley/Biped/Boots/Boots_Small/Wavs/Mimosa_Biped_Boots_Small_Scuff_03}
 }
 {
  id=WavePlayer_14
  properties {soundWaveAssetPtr=/CosmeticCompanions/Assets/Audio_Shared/Foley/Biped/Boots/Boots_Small/Wavs/Mimosa_Biped_Boots_Small_Scuff_04}
 }
 {
  id=WavePlayer_15
  properties {soundWaveAssetPtr=/CosmeticCompanions/Assets/Audio_Shared/Foley/Biped/Boots/Boots_Small/Wavs/Mimosa_Biped_Boots_Small_Scuff_05}
 }
 {
  id=WavePlayer_16
  properties {soundWaveAssetPtr=/CosmeticCompanions/Assets/Audio_Shared/Foley/Biped/Boots/Boots_Small/Wavs/Mimosa_Biped_Boots_Small_Scuff_06}
 }
 {
  id=WavePlayer_20
  properties {soundWaveAssetPtr=/CosmeticCompanions/Assets/Audio_Shared/Foley/Biped/Boots/Boots_Small/Wavs/Mimosa_Biped_Boots_Small_Scuff_07}
 }
 {
  id=WavePlayer_21
  properties {soundWaveAssetPtr=/CosmeticCompanions/Assets/Audio_Shared/Foley/Biped/Boots/Boots_Small/Wavs/Mimosa_Biped_Boots_Small_Scuff_08}
 }
 {
  id=WavePlayer_22
  properties {soundWaveAssetPtr=&Mimosa_Biped_Boots_Small_Scuff_01}
 }
 {
  id=WavePlayer_23
  properties {soundWaveAssetPtr=&Mimosa_Biped_Boots_Small_Scuff_02}
 }
 {
  id=WavePlayer_24
  properties {soundWaveAssetPtr=&Mimosa_Biped_Boots_Small_Scuff_03}
 }
 {
  id=WavePlayer_25
  properties {soundWaveAssetPtr=&Mimosa_Biped_Boots_Small_Scuff_04}
 }
]
properties {
 attenuationSettings=/Game/Sounds/Attenuation/AS_Companion_Foley
 maxDistance=1200
 priority=0
 soundClassObject=&Fort_Creatures_Pets
 volumeMultiplier=0.3
 concurrencySet [/CosmeticCompanions/Assets/Audio_Shared/Foley/Biped/CONC_Mimosa_Biped_Footstep]
}
